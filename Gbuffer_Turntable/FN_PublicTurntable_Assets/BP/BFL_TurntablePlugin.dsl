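; Blueprint function library for the GBuffer Turntable plugin, providing asset lookup and sequence scalability helpers.
; One function resolves a soft object reference through the Asset Registry, returning a validity flag and the asset.
; Another reads the Scalability string from a placed BP_Turntable_GBuffers actor, applies EPIC or LOW and returns a code.
; The setter runs the SCALABILITY console command, and its warning always says EPIC, even for LOW.
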(fn TT_Func_GetAssetIfValid
  (params (SoftObjectReference :type Soft%20Object%20Reference%20%28by%20ref%29))
  (bind _returnvalue (AssetRegistry|GetAssetByObjectPath (AssetRegistry|GetAssetRegistry) (Utilities|String|StringToName (Utilities|ToString(SoftObjectReference) SoftObjectReference)) false))
  (if (AssetData|IsValidAssetData _returnvalue)
    (return true (AssetData|GetAsset _returnvalue))
    (else
      (return false 0))))

(fn TT_FUNC_Sequence_SetScalability
  (bind _outactors (Actor|GetAllActorsOfClass 0 "/Gbuffer_Turntable/FN_PublicTurntable_Assets/BP/BP_Turntable_GBuffers.BP_Turntable_GBuffers_C"))
  (if (Utilities|Array|IsNotEmpty _outactors)
    (switch string (Variables|/Gbuffer_Turntable/FN_PublicTurntable_Assets/BP/BP_Turntable_GBuffers.BP_Turntable_GBuffers_C|GetScalability (Utilities|Array|Get(acopy) _outactors))
      (:Default
        (Development|PrintString 0 "ERROR IN THE SEQUENCE - Scalability"))
      (:EPIC
        (CallFunction|TTFUNCSetScalability 3)
        (return 2))
      (:LOW
        (CallFunction|TTFUNCSetScalability 0)
        (return 1))
      (:Default
        (return 0)))))

(fn TT_FUNC_SetScalability
  (params (Scalability :type Integer))
  (Development|ExecuteConsoleCommand 0 (Utilities|String|Append "SCALABILITY " (Utilities|String|ToString(Integer) Scalability)))
  (Development|PrintString 0 "WARNING! ~~~~ Scalability Forced to EPIC quality for sequence!" true true "(R=1.000000,G=0.000000,B=0.000000,A=1.000000)" 4.0))
