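; Turntable Blueprint that feeds its Arrow's direction to Fortnite's Material Parameter Collection as the toon sun light direction.
; Both the construction script and BeginPlay resolve the collection through a plugin helper, continuing only if it is valid.
; They write the negated Arrow forward vector to SunlightDirectionalVector and set bEnableCharacterRimLight to its unset default, zero.
; The update runs only on construction and at play start, with no per-frame refresh.
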(event EventBeginPlay
  (bind (_is_valid _returnvalue) (Class|BFLTurntablePlugin|TTFuncGetAssetifValid (Variables|Default|GetFortniteMPC)))
  (if _is_valid
    (Utilities|IsValid _returnvalue
      (:Is%20Valid
        (Rendering|Material|SetVectorParameterValue (Utilities|Casting|CastToMaterialParameterCollection _returnvalue) (Utilities|Name|MakeLiteralName "SunlightDirectionalVector") (Math|Conversions|ToLinearColor(Vector) (* (Transformation|GetForwardVector (Variables|Default|GetArrow)) -1.0)) :!class /Script/Engine.KismetMaterialLibrary)
        (Rendering|Material|SetScalarParameterValue (Utilities|Casting|CastToMaterialParameterCollection _returnvalue) (Utilities|Name|MakeLiteralName "bEnableCharacterRimLight") :!class /Script/Engine.KismetMaterialLibrary)))))

(fn ConstructionScript
  ;[1 1 1] Check and set FN Engine Specific Value
  (bind (_is_valid _returnvalue) (Class|BFLTurntablePlugin|TTFuncGetAssetifValid (Variables|Default|GetFortniteMPC)))
  (if _is_valid
    (Utilities|IsValid _returnvalue
      (:Is%20Valid
        (Rendering|Material|SetVectorParameterValue (Utilities|Casting|CastToMaterialParameterCollection _returnvalue) (Utilities|Name|MakeLiteralName "SunlightDirectionalVector") (Math|Conversions|ToLinearColor(Vector) (* (Transformation|GetForwardVector (Variables|Default|GetArrow)) -1.0)) :!class /Script/Engine.KismetMaterialLibrary)
        (Rendering|Material|SetScalarParameterValue (Utilities|Casting|CastToMaterialParameterCollection _returnvalue) (Utilities|Name|MakeLiteralName "bEnableCharacterRimLight") :!class /Script/Engine.KismetMaterialLibrary)))))
