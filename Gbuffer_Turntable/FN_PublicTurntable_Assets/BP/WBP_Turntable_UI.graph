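# Movie Render Pipeline burn-in widget that overlays turntable renders with name, date, build, object list, pass, frame number, scalability and version text.
# setText fills all fields, while setPass and setFrameNumber update single ones; the version string is hardcoded, leaving the version variable unused.
# Tick reads Lumen GI and reflection CVars and reveals a red warning if either is on, never re-hiding it.
# Color-coded checker legend labels are static; their hide node is unconnected.
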WidgetBlueprint /Gbuffer_Turntable/FN_PublicTurntable_Assets/BP/WBP_Turntable_UI
graphs [
 EventGraph {nodes [
   Event Tick()
   > n1 = GetConsoleVariableBoolValue!(execute, VariableName:String=r.Lumen.DiffuseIndirect.Allow) -> (then, ReturnValue:bool)
   > n2 = GetConsoleVariableBoolValue(VariableName=r.Lumen.Reflections.Allow)
   > Branch(Condition=OR Boolean(A=@n2, B=@n1))
   > SetVisibility(self=Get LumenWarning() -> (LumenWarning:Text Object Reference), InVisibility=Visible)
   SetVisibility(self=Get CheckerControlBox() -> (CheckerControlBox:Vertical Box Object Reference), InVisibility=Hidden)
   Event PreConstruct()
   Event Construct()
  ]}
 setFrameNumber {nodes [
   n1 = setFrameNumber!() -> (then, frameNumber:int) {is=title K2Node_FunctionEntry}
   > "SetText (Text)"(self=Get frameNumber() -> (frameNumber:Text Object Reference), InText="To Text (Integer)"(Value=@n1))
  ]}
 setPass {nodes [
   n1 = setPass!() -> (then, Pass:String) {is=title K2Node_FunctionEntry}
   > "SetText (Text)"(self=Get Pass() -> (Pass:Text Object Reference), InText="To Text (String)"(InString=@n1))
  ]}
 setText {nodes [
   n1 = setText!() -> (then, Name:String, Date:String, build:String, objectList:String, Pass:String, FrameNumber:String, Scalability:String) {is=title K2Node_FunctionEntry}
   > Sequence()
   > "SetText (Text)"(self=Get Name() -> (Name:Text Object Reference), InText="To Text (String)"(InString=@n1))
   > "SetText (Text)"(self=Get Date() -> (Date:Text Object Reference), InText="To Text (String)"(InString=@n1.Date))
   > "SetText (Text)"(self=Get BuildNumber() -> (BuildNumber:Text Object Reference), InText="To Text (String)"(InString=@n1.build))
   > "SetText (Text)"(self=Get objectlist() -> (objectlist:Text Object Reference), InText="To Text (String)"(InString=@n1.objectList))
   > "SetText (Text)"(self=Get Pass(), InText="To Text (String)"(InString=@n1.Pass))
   > "SetText (Text)"(self=Get frameNumber(), InText="To Text (String)"(InString=@n1.FrameNumber))
   > "SetText (Text)"(
    self=Get PluginVersion() -> (PluginVersion:Text Object Reference)
    InText="To Text (String)"(InString=Append(A="Turntable Version: ", B=VC - 1.0))
   )
   > "SetText (Text)"(
    self=Get Scalability() -> (Scalability:Text Object Reference)
    InText="To Text (String)"(InString=Append(A="Scalability: ", B=@n1.Scalability))
   )
   "To String (Float)"(InDouble=Get!() -> (Current Plugin Information_CurrentVersion_3_#1:float=~))
  ]}
]
variables [Current Plugin Information {type=GbufferTT_VersionInfo, default {current Version=2.4}}]
widgets {
 info {
  inheritedWidgetCount=0
  parentClass=/Script/MovieRenderPipelineSettings.MoviePipelineBurnInWidget
  rootWidgetClass=&CanvasPanel
  widgetCount=30
 }
 widgets [
  CanvasPanel_0 {widgetClassPath=&CanvasPanel, widget=CanvasPanel_0}
  Border_8 {
   widgetClassPath=&Border
   parent=CanvasPanel_0
   widget=Border_8
   properties {brushColor=<rgba 0.2 0.2 0.2 1>}
   slot {
    class=CanvasPanelSlot
    properties {bAutoSize=true, layoutData {alignment=<0 0>, anchors {maximum=<0 0>, minimum=<0 0>}, offsets=<ltrb 20 108 800 50>}}
   }
  }
  Name {
   widgetClassPath=&TextBlock
   bIsVariable=true
   parent=Border_8
   widget=Name
   properties {
    text=Name
    font {
     bForceMonospaced=false
     bMaterialIsStencil=false
     fontMaterial=null
     fontObject=&Roboto
     letterSpacing=0
     monospacedWidth=1
     size=30
     skewAmount=0
     typefaceFontName=Bold
     outlineSettings {
      bApplyOutlineToDropShadows=false
      bMiteredCorners=false
      bSeparateFillAlpha=false
      outlineMaterial=null
      outlineSize=0
      outlineColor=<rgba 0 0 0 1>
     }
    }
   }
   slot {class=BorderSlot}
  }
  Border_7 {
   widgetClassPath=&Border
   parent=CanvasPanel_0
   widget=Border_7
   properties {brushColor=<rgba 0.2 0.2 0.2 1>}
   slot {
    class=CanvasPanelSlot
    properties {bAutoSize=true, layoutData {alignment=<0 0>, anchors {maximum=<0 0>, minimum=<0 0>}, offsets=<ltrb 20 164 303.5035 50>}}
   }
  }
  PluginVersion {
   widgetClassPath=&TextBlock
   bIsVariable=true
   parent=Border_7
   widget=PluginVersion
   properties {
    text=Version
    font {
     bForceMonospaced=false
     bMaterialIsStencil=false
     fontMaterial=null
     fontObject=&Roboto
     letterSpacing=0
     monospacedWidth=1
     size=20
     skewAmount=0
     typefaceFontName=Bold
     outlineSettings {
      bApplyOutlineToDropShadows=false
      bMiteredCorners=false
      bSeparateFillAlpha=false
      outlineMaterial=null
      outlineSize=0
      outlineColor=<rgba 0 0 0 1>
     }
    }
   }
   slot {class=BorderSlot}
  }
  Border_6 {
   widgetClassPath=&Border
   parent=CanvasPanel_0
   widget=Border_6
   properties {brushColor=<rgba 0.2 0.2 0.2 1>}
   slot {
    class=CanvasPanelSlot
    properties {bAutoSize=true, layoutData {alignment=<0 0>, anchors {maximum=<0 0>, minimum=<0 0>}, offsets=<ltrb 20 208 306.7067 50>}}
   }
  }
  Scalability {
   widgetClassPath=&TextBlock
   bIsVariable=true
   parent=Border_6
   widget=Scalability
   properties {
    text=Scalability
    font {
     bForceMonospaced=false
     bMaterialIsStencil=false
     fontMaterial=null
     fontObject=&Roboto
     letterSpacing=0
     monospacedWidth=1
     size=20
     skewAmount=0
     typefaceFontName=Bold
     outlineSettings {
      bApplyOutlineToDropShadows=false
      bMiteredCorners=false
      bSeparateFillAlpha=false
      outlineMaterial=null
      outlineSize=0
      outlineColor=<rgba 0 0 0 1>
     }
    }
   }
   slot {class=BorderSlot}
  }
  Border_5 {
   widgetClassPath=&Border
   parent=CanvasPanel_0
   widget=Border_5
   properties {brushColor=<rgba 0.2 0.2 0.2 1>}
   slot {
    class=CanvasPanelSlot
    properties {bAutoSize=true, layoutData {alignment=<0 0>, anchors {maximum=<0 0>, minimum=<0 0>}, offsets=<ltrb 20 256 306.7067 50>}}
   }
  }
  LumenWarning {
   widgetClassPath=&TextBlock
   bIsVariable=true
   parent=Border_5
   widget=LumenWarning
   properties {
    text="LUMEN IS ENABLED!"
    visibility=Hidden
    colorAndOpacity {colorUseRule=UseColor_Specified, specifiedColor=<rgba 1 0 0 1>}
    font {
     bForceMonospaced=false
     bMaterialIsStencil=false
     fontMaterial=null
     fontObject=&Roboto
     letterSpacing=0
     monospacedWidth=1
     size=20
     skewAmount=0
     typefaceFontName=Bold
     outlineSettings {
      bApplyOutlineToDropShadows=false
      bMiteredCorners=false
      bSeparateFillAlpha=false
      outlineMaterial=null
      outlineSize=0
      outlineColor=<rgba 0 0 0 1>
     }
    }
   }
   slot {class=BorderSlot}
  }
  CheckerControlBox {
   widgetClassPath=&VerticalBox
   bIsVariable=true
   parent=CanvasPanel_0
   widget=CheckerControlBox
   slot {class=CanvasPanelSlot, properties {layoutData {alignment=<0 0>, anchors {maximum=<0 0>, minimum=<0 0>}, offsets=<ltrb 16 324 309.9099 50>}}}
  }
  Border_4 {
   widgetClassPath=&Border
   parent=CheckerControlBox
   widget=Border_4
   properties {brushColor=<rgba 0.2 0.2 0.2 1>}
   slot {class=VerticalBoxSlot}
  }
  Checker_MinColor {
   widgetClassPath=&TextBlock
   bIsVariable=true
   parent=Border_4
   widget=Checker_MinColor
   properties {
    text=Min Color Value
    colorAndOpacity {colorUseRule=UseColor_Specified, specifiedColor=<rgba 1 1 0 1>}
    font {
     bForceMonospaced=false
     bMaterialIsStencil=false
     fontMaterial=null
     fontObject=&Roboto
     letterSpacing=0
     monospacedWidth=1
     size=24
     skewAmount=0
     typefaceFontName=Bold
     outlineSettings {
      bApplyOutlineToDropShadows=false
      bMiteredCorners=false
      bSeparateFillAlpha=false
      outlineMaterial=null
      outlineSize=1
      outlineColor=<rgba 0 0 0 1>
     }
    }
   }
   slot {class=BorderSlot}
  }
  Border_3 {
   widgetClassPath=&Border
   parent=CheckerControlBox
   widget=Border_3
   properties {brushColor=<rgba 0.2 0.2 0.2 1>}
   slot {class=VerticalBoxSlot}
  }
  Checker_MaxColor {
   widgetClassPath=&TextBlock
   bIsVariable=true
   parent=Border_3
   widget=Checker_MaxColor
   properties {
    text=Max Color Value
    colorAndOpacity {colorUseRule=UseColor_Specified, specifiedColor=<rgba 1 0 1 1>}
    font {
     bForceMonospaced=false
     bMaterialIsStencil=false
     fontMaterial=null
     fontObject=&Roboto
     letterSpacing=0
     monospacedWidth=1
     size=24
     skewAmount=0
     typefaceFontName=Bold
     outlineSettings {
      bApplyOutlineToDropShadows=false
      bMiteredCorners=false
      bSeparateFillAlpha=false
      outlineMaterial=null
      outlineSize=1
      outlineColor=<rgba 0 0 0 1>
     }
    }
   }
   slot {class=BorderSlot}
  }
  Border_2 {
   widgetClassPath=&Border
   parent=CheckerControlBox
   widget=Border_2
   properties {brushColor=<rgba 0.2 0.2 0.2 1>}
   slot {class=VerticalBoxSlot}
  }
  Checker_MinMetal {
   widgetClassPath=&TextBlock
   bIsVariable=true
   parent=Border_2
   widget=Checker_MinMetal
   properties {
    text=Min Metallic Value
    colorAndOpacity {colorUseRule=UseColor_Specified, specifiedColor=<rgba 1 0 0 1>}
    font {
     bForceMonospaced=false
     bMaterialIsStencil=false
     fontMaterial=null
     fontObject=&Roboto
     letterSpacing=0
     monospacedWidth=1
     size=24
     skewAmount=0
     typefaceFontName=Bold
     outlineSettings {
      bApplyOutlineToDropShadows=false
      bMiteredCorners=false
      bSeparateFillAlpha=false
      outlineMaterial=null
      outlineSize=1
      outlineColor=<rgba 0 0 0 1>
     }
    }
   }
   slot {class=BorderSlot}
  }
  Border_1 {
   widgetClassPath=&Border
   parent=CheckerControlBox
   widget=Border_1
   properties {brushColor=<rgba 0.2 0.2 0.2 1>}
   slot {class=VerticalBoxSlot}
  }
  Checker_MinSpec {
   widgetClassPath=&TextBlock
   bIsVariable=true
   parent=Border_1
   widget=Checker_MinSpec
   properties {
    text=Min Specular Value
    colorAndOpacity {colorUseRule=UseColor_Specified, specifiedColor=<rgba 0 0 1 1>}
    font {
     bForceMonospaced=false
     bMaterialIsStencil=false
     fontMaterial=null
     fontObject=&Roboto
     letterSpacing=0
     monospacedWidth=1
     size=24
     skewAmount=0
     typefaceFontName=Bold
     outlineSettings {
      bApplyOutlineToDropShadows=false
      bMiteredCorners=false
      bSeparateFillAlpha=false
      outlineMaterial=null
      outlineSize=1
      outlineColor=<rgba 0 0 0 1>
     }
    }
   }
   slot {class=BorderSlot}
  }
  Border_22 {
   widgetClassPath=&Border
   parent=CanvasPanel_0
   widget=Border_22
   properties {brushColor=<rgba 0.2 0.2 0.2 1>}
   slot {
    class=CanvasPanelSlot
    properties {bAutoSize=true, layoutData {alignment=<0 0>, anchors {maximum=<0 0.4548>, minimum=<0 0.4548>}, offsets=<ltrb 20 16.3081 590.1902 379.3616>}}
   }
  }
  objectlist {
   widgetClassPath=&TextBlock
   bIsVariable=true
   parent=Border_22
   widget=objectlist
   properties {
    text=Objects
    font {
     bForceMonospaced=false
     bMaterialIsStencil=false
     fontMaterial=null
     fontObject=&Roboto
     letterSpacing=0
     monospacedWidth=1
     size=20
     skewAmount=0
     typefaceFontName=Bold
     outlineSettings {
      bApplyOutlineToDropShadows=false
      bMiteredCorners=false
      bSeparateFillAlpha=false
      outlineMaterial=null
      outlineSize=0
      outlineColor=<rgba 0 0 0 1>
     }
    }
   }
   slot {class=BorderSlot}
  }
  Border_0 {
   widgetClassPath=&Border
   parent=CanvasPanel_0
   widget=Border_0
   properties {brushColor=<rgba 0.2 0.2 0.2 1>}
   slot {
    class=CanvasPanelSlot
    properties {bAutoSize=true, layoutData {alignment=<0 0>, anchors {maximum=<0 1>, minimum=<0 1>}, offsets=<ltrb 20 -195 500 50>}}
   }
  }
  Pass {
   widgetClassPath=&TextBlock
   bIsVariable=true
   parent=Border_0
   widget=Pass
   properties {
    text=PassName
    font {
     bForceMonospaced=false
     bMaterialIsStencil=false
     fontMaterial=null
     fontObject=&Roboto
     letterSpacing=0
     monospacedWidth=1
     size=30
     skewAmount=0
     typefaceFontName=Bold
     outlineSettings {
      bApplyOutlineToDropShadows=false
      bMiteredCorners=false
      bSeparateFillAlpha=false
      outlineMaterial=null
      outlineSize=0
      outlineColor=<rgba 0 0 0 1>
     }
    }
   }
   slot {class=BorderSlot}
  }
  Border_21 {
   widgetClassPath=&Border
   parent=CanvasPanel_0
   widget=Border_21
   properties {brushColor=<rgba 0.2 0.2 0.2 1>}
   slot {class=CanvasPanelSlot, properties {layoutData {alignment=<0 0>, anchors {maximum=<1 1>, minimum=<1 1>}, offsets=<ltrb -140 -155 120 50>}}}
  }
  frameNumber {
   widgetClassPath=&TextBlock
   bIsVariable=true
   parent=Border_21
   widget=frameNumber
   properties {
    justification=Right
    text=9999
    font {
     bForceMonospaced=false
     bMaterialIsStencil=false
     fontMaterial=null
     fontObject=&Roboto
     letterSpacing=0
     monospacedWidth=1
     size=30
     skewAmount=0
     typefaceFontName=Bold
     outlineSettings {
      bApplyOutlineToDropShadows=false
      bMiteredCorners=false
      bSeparateFillAlpha=false
      outlineMaterial=null
      outlineSize=0
      outlineColor=<rgba 0 0 0 1>
     }
    }
   }
   slot {class=BorderSlot}
  }
  Border_20 {
   widgetClassPath=&Border
   parent=CanvasPanel_0
   widget=Border_20
   properties {horizontalAlignment=HAlign_Right, verticalAlignment=VAlign_Bottom, brushColor=<rgba 0.2 0.2 0.2 1>}
   slot {
    class=CanvasPanelSlot
    properties {bAutoSize=true, layoutData {alignment=<0 0>, anchors {maximum=<0 1>, minimum=<0 1>}, offsets=<ltrb 20 -240 500 60.02>}}
   }
  }
  BuildNumber {
   widgetClassPath=&TextBlock
   bIsVariable=true
   parent=Border_20
   widget=BuildNumber
   properties {
    justification=Right
    text=Build
    font {
     bForceMonospaced=false
     bMaterialIsStencil=false
     fontMaterial=null
     fontObject=&Roboto
     letterSpacing=0
     monospacedWidth=1
     size=20
     skewAmount=0
     typefaceFontName=Bold
     outlineSettings {
      bApplyOutlineToDropShadows=false
      bMiteredCorners=false
      bSeparateFillAlpha=false
      outlineMaterial=null
      outlineSize=0
      outlineColor=<rgba 0 0 0 1>
     }
    }
   }
   slot {class=BorderSlot, properties {horizontalAlignment=HAlign_Right, verticalAlignment=VAlign_Bottom}}
  }
  Border_19 {
   widgetClassPath=&Border
   parent=CanvasPanel_0
   widget=Border_19
   properties {horizontalAlignment=HAlign_Right, brushColor=<rgba 0.2 0.2 0.2 1>}
   slot {
    class=CanvasPanelSlot
    properties {bAutoSize=true, layoutData {alignment=<0 0>, anchors {maximum=<0 0>, minimum=<0 0>}, offsets=<ltrb 20 40 500 50>}}
   }
  }
  Date {
   widgetClassPath=&TextBlock
   bIsVariable=true
   parent=Border_19
   widget=Date
   properties {
    justification=Right
    text=Date
    font {
     bForceMonospaced=false
     bMaterialIsStencil=false
     fontMaterial=null
     fontObject=&Roboto
     letterSpacing=0
     monospacedWidth=1
     size=30
     skewAmount=0
     typefaceFontName=Bold
     outlineSettings {
      bApplyOutlineToDropShadows=false
      bMiteredCorners=false
      bSeparateFillAlpha=false
      outlineMaterial=null
      outlineSize=0
      outlineColor=<rgba 0 0 0 1>
     }
    }
   }
   slot {class=BorderSlot, properties {horizontalAlignment=HAlign_Right}}
  }
  Border_18 {
   widgetClassPath=&Border
   parent=CanvasPanel_0
   widget=Border_18
   properties {brushColor=<rgba 0.2 0.2 0.2 1>}
   slot {
    class=CanvasPanelSlot
    properties {bAutoSize=true, layoutData {alignment=<0 0>, anchors {maximum=<0 1>, minimum=<0 1>}, offsets=<ltrb 24 -137 500 32.84>}}
   }
  }
  TextBlock_1 {
   widgetClassPath=&TextBlock
   bIsVariable=true
   parent=Border_18
   widget=TextBlock_1
   properties {
    justification=Right
    text=Fortnite v4.0
    font {
     bForceMonospaced=false
     bMaterialIsStencil=false
     fontMaterial=null
     fontObject=&Roboto
     letterSpacing=0
     monospacedWidth=1
     size=20
     skewAmount=0
     typefaceFontName=Bold
     outlineSettings {
      bApplyOutlineToDropShadows=false
      bMiteredCorners=false
      bSeparateFillAlpha=false
      outlineMaterial=null
      outlineSize=0
      outlineColor=<rgba 0 0 0 1>
     }
    }
   }
   slot {class=BorderSlot}
  }
 ]
}
