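# Post-process material for the Gbuffer turntable that flags base color values outside an allowed brightness range.
# It fully desaturates the stored base color and compares it to MinValue and MaxValue parameters, painting out-of-range pixels in configurable warning colors (yellow for low, magenta for high by default).
# In-range pixels, and those with a shading model ID at or below a small threshold, keep the scene color.
# A gamma-like Power node on base color feeds nothing shown.
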Material /Gbuffer_Turntable/FN_PublicTurntable_Assets/Materials/MAT_Turntable_BaseColorLevelsChecker_Viewport
graph {
 properties {materialDomain=MD_PostProcess, pixelDepthOffsetMode=PDOM_Legacy, usedShadingModels=Default Lit}
 nodes [
  n1 = MaterialFunctionCall(Color=SceneTexture() {sceneTextureId=PPI_StoredBaseColor}.Color) {materialFunction=&SplitComponents}
  n2 = MaterialFunctionCall(Float3=Desaturation(None=@n1.RGB, Fraction=Constant() {r=1})) {materialFunction=&BreakOutFloat3Components}
  n3 = MaterialFunctionCall(Color=SceneTexture() {sceneTextureId=PPI_PostProcessInput0}.Color) {materialFunction=&SplitComponents}
  n4 = If(
   A=@n2.R
   B=ScalarParameter() {defaultValue=0.9, parameterName=MaxValue}
   "A > B"=VectorParameter() {parameterName=MaxValueColor, defaultValue=<rgba 1 0 0.8328 1>}.RGB
   "A == B"=@n3.RGB
   "A < B"=@n3.RGB
  )
  If(
   A=MaterialFunctionCall(Color=SceneTexture() {sceneTextureId=PPI_ShadingModelID}.Color) {materialFunction=&SplitComponents}.R
   B=Constant() {r=0.1}
   "A > B"=If(
    A=@n2.R
    B=ScalarParameter() {defaultValue=0.009, parameterName=MinValue}
    "A > B"=@n4
    "A == B"=@n4
    "A < B"=VectorParameter() {parameterName=MinValueColor, defaultValue=<rgba 1 1 0 1>}.RGB
   )
   "A == B"=@n3.RGB
   "A < B"=@n3.RGB
  )
  Power(Base=@n1.RGB, Exp=Constant() {r=0.45})
 ]
}
