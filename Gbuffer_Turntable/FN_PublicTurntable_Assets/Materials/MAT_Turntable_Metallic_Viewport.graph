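# Post-process material for a GBuffer turntable that, going by its name, displays the metallic channel in the viewport.
# The graph samples the Metallic scene texture and feeds its color into a Power node.
# The fixed exponent approximates a linear-to-gamma (1/2.2) conversion of the raw buffer values.
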Material /Gbuffer_Turntable/FN_PublicTurntable_Assets/Materials/MAT_Turntable_Metallic_Viewport
graph {
 properties {materialDomain=MD_PostProcess, pixelDepthOffsetMode=PDOM_Legacy, usedShadingModels=Default Lit}
 nodes [Power(Base=SceneTexture() {sceneTextureId=PPI_Metallic}.Color) {constExponent=0.45}]
}
