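# M_Br_Background_Overlay is a translucent Default Lit material in the turntable assets, apparently meant as a tintable background overlay.
# It samples SceneColor, splits the channels and takes their maximum using only red and blue, as red feeds twice and green is unused.
# That brightness value is multiplied by the BackgroundColor parameter, white by default.
# The file shows no link from this result or a stray constant to any material output.
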Material /Gbuffer_Turntable/FN_PublicTurntable_Assets/Materials/M_Br_Background_Overlay
graph {
 properties {blendMode=BLEND_Translucent, pixelDepthOffsetMode=PDOM_Legacy, usedShadingModels=Default Lit}
 nodes [
  Constant() {r=0.5}
  n1 = MaterialFunctionCall(Float3=SceneColor!(UVs).RGB) {materialFunction=&BreakOutFloat3Components}
  Multiply(A=Max(A=Max(A=@n1.R, B=@n1.R), B=@n1.B), B=VectorParameter() {parameterName=BackgroundColor, defaultValue=<rgba 1 1 1 0>}.RGB)
 ]
}
