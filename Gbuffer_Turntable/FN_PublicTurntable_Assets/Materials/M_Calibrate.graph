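# M_Calibrate is a simple Default Lit material in the Gbuffer Turntable assets, apparently for calibration, judging by its name.
# It holds a Color vector parameter with an orange-brown default, which material instances can override, and a mid-value scalar constant.
# The graph shows no links from either node to material outputs, so their roles in shading are not specified.
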Material /Gbuffer_Turntable/FN_PublicTurntable_Assets/Materials/M_Calibrate
graph {
 properties {pixelDepthOffsetMode=PDOM_Legacy, usedShadingModels=Default Lit}
 nodes [
  Constant() {r=0.5}
  VectorParameter() {parameterName=Color, defaultValue=<rgba 0.7 0.5 0.1 0>}
 ]
}
