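# Post-process material that isolates the separate translucency alpha, named for use as a Movie Render Queue pass in the Gbuffer turntable.
# It reads the PostProcessInput1 scene texture and masks out its alpha channel.
# A feature level switch passes that alpha by default, substituting a constant one on the ES2 and ES3.1 feature levels.
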Material /Gbuffer_Turntable/FN_PublicTurntable_Assets/Materials/M_SeparateTranslucencyA_MRQ
graph {
 properties {materialDomain=MD_PostProcess, pixelDepthOffsetMode=PDOM_Legacy, shadingModel=MSM_Unlit, usedShadingModels=Unlit}
 nodes [
  n1 = ComponentMask(None=SceneTexture() {sceneTextureId=PPI_PostProcessInput1}.Color) {a=true}
  n2 = Constant() {r=1}
  FeatureLevelSwitch(Default=@n1, ES2=@n2, ES3_1=@n2, SM4_REMOVED=@n1)
 ]
}
