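# Post-process material in the Gbuffer turntable assets that builds a depth-based alpha cutout mask, with alpha output enabled.
# Scene depth is divided by a fixed threshold, floored, saturated and inverted, giving a hard binary mask of one for nearer pixels and zero beyond.
# The graph also holds a post-process input scene texture and a constant color, though their connections to the material outputs are not shown.
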Material /Gbuffer_Turntable/FN_PublicTurntable_Assets/Materials/M_Turntable_Alphacutout
graph {
 properties {bAutomaticallySetUsageInEditor=false, blendableOutputAlpha=true, materialDomain=MD_PostProcess, usedShadingModels=Default Lit}
 nodes [
  Constant3Vector()
  n1 = Divide(A=SceneDepth().R) {constB=2000}
  OneMinus(None=Saturate(None=Floor(None=@n1)))
  SceneTexture() {sceneTextureId=PPI_PostProcessInput0}
 ]
}
