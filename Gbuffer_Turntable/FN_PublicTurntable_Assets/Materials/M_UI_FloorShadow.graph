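# Default Lit floor material for the turntable assets, with a tunable radial color and surface falloff.
# A UV-based exponential radial gradient masks floor color, specular and roughness values.
# A second gradient on screen-aligned UVs, with tile and offset parameters, blends two colors.
# Its resolution ratio divides the X component by itself, so it applies no aspect correction.
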Material /Gbuffer_Turntable/FN_PublicTurntable_Assets/Materials/M_UI_FloorShadow
graph {
 properties {pixelDepthOffsetMode=PDOM_Legacy, usedShadingModels=Default Lit}
 nodes [
  n1 = MaterialFunctionCall(
   UVs=TextureCoordinate()
   Radius=ScalarParameter() {defaultValue=0.25, parameterName=Radius Mask}
   Density=ScalarParameter() {defaultValue=0.45, parameterName=Density Mask}
  ) {materialFunction=&RadialGradientExponential}
  LinearInterpolate(B=ScalarParameter() {defaultValue=4, parameterName=Spec}, Alpha=@n1.RadialGradientExponential)
  LinearInterpolate(B=ScalarParameter() {defaultValue=1, parameterName=Roughness}, Alpha=@n1.RadialGradientExponential) {constA=1}
  LinearInterpolate(B=VectorParameter() {parameterName=FloorRGB, defaultValue=<rgba 0.0467 0.3467 0.956 1>}.RGB, Alpha=@n1.RadialGradientExponential)
  n2 = AppendVector(A=ScalarParameter() {parameterName=x-Offset}, B=ScalarParameter() {parameterName=y-Offset})
  n3 = MaterialFunctionCall(Float2=MaterialFunctionCall() {materialFunction=&ScreenResolution}.Visible Resolution) {materialFunction=&BreakOutFloat2Components}
  n4 = Divide(A=@n3.R, B=@n3.R)
  n5 = AppendVector(A=Multiply(A=@n4) {constB=0.5}, B=Constant() {r=0.5})
  n6 = Multiply(
   A=AppendVector(A=@n4, B=Constant() {r=1})
   B=MaterialFunctionCall() {materialFunction=&ScreenAlignedUVs}."X 100%, Y 100%"
  )
  n7 = ScalarParameter() {defaultValue=1, parameterName=Tile}
  LinearInterpolate(
   A=ComponentMask(None=VectorParameter() {parameterName=RGBA0, defaultValue=<rgba 0.03 0.0525 0.3 1>}.RGB) {b=true, g=true, r=true}
   B=ComponentMask(None=VectorParameter() {parameterName=RGBA1, defaultValue=<rgba 0.0475 0.3483 0.95 1>}.RGB) {b=true, g=true, r=true}
   Alpha=MaterialFunctionCall(
    UVs=Multiply(A=@n6, B=@n7)
    CenterPosition=Add(A=@n5, B=@n2)
    Radius=ScalarParameter() {defaultValue=1, parameterName=Radius}
    Density=ScalarParameter() {defaultValue=1.125, parameterName=Density}
   ) {materialFunction=&RadialGradientExponential}.RadialGradientExponential
  )
  ScalarParameter() {parameterName=Metalic}
 ]
}
